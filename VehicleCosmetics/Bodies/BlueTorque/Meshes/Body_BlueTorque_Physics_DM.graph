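# Physics asset for the BlueTorque vehicle body cosmetic mesh, providing its simplified collision shape.
# A single body on the root bone holds one box, raised and shifted slightly forward to enclose the car hull.
# It has no constraints, so the body acts as one rigid piece with no articulated parts.
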PhysicsAsset /VehicleCosmetics/Bodies/BlueTorque/Meshes/Body_BlueTorque_Physics_DM
bodies {root [{shapeType=Box, extentX=444, extentY=216, extentZ=132, shapeName=root_box, center=<24.2495 0 102.4696>}]}
constraints []
